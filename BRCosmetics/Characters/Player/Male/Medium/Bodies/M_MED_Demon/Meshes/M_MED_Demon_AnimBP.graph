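# Post-copy anim blueprint for the M_MED_Demon cosmetic body, which follows the attached parent mesh's pose and adds procedural fixes for clothing and backpack bones.
# Each update maps upperarm rotation relative to its clavicle into a per-side alpha.
# Those alphas drive additive bone-space offsets on dyn_clothes_l and dyn_clothes_r.
# A weighted constraint then blends dyn_backpack between both clavicles and upper spine bones.
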AnimBlueprint /BRCosmetics/Characters/Player/Male/Medium/Bodies/M_MED_Demon/Meshes/M_MED_Demon_AnimBP
target_skeleton=&/M_MED_Demon_Skeleton
graphs [
 AnimGraph {nodes [
   n1 = "Transform (Modify) Bone - Bone: dyn_clothes_l"(
    ComponentPose=Local To Component(LocalPose=Copy Pose From Mesh() {bUseAttachedParent=true})
    Alpha=Get Shoulder_rot_f_l() -> (Shoulder_rot_f_l:double)
    Translation="-5,0,1"
    Rotation="0,0,0"
    Scale="1,1,1"
   ) {
    rotationMode=BMM_Additive
    rotationSpace=BCS_BoneSpace
    scaleMode=BMM_Additive
    scaleSpace=BCS_BoneSpace
    translationMode=BMM_Additive
    translationSpace=BCS_BoneSpace
    alphaBoolBlend {blendOption=Linear}
    boneToModify {boneName=dyn_clothes_l}
    translation {x=-5, z=1}
   }
   n2 = "Transform (Modify) Bone - Bone: dyn_clothes_r"(ComponentPose=@n1, Alpha=Get Shoulder_rot_f_r() -> (Shoulder_rot_f_r:double), Translation="5,0,-1", Rotation="0,0,0", Scale="1,1,1") {
    rotationMode=BMM_Additive
    rotationSpace=BCS_BoneSpace
    scaleMode=BMM_Additive
    scaleSpace=BCS_BoneSpace
    translationMode=BMM_Additive
    translationSpace=BCS_BoneSpace
    alphaBoolBlend {blendOption=Linear}
    boneToModify {boneName=dyn_clothes_r}
    translation {x=5, z=-1}
   }
   n3 = "Constraint - dyn_backpack (PPPPP)"!(
    ComponentPose=@n2
    Alpha=1
    ConstraintWeights_0=0.2133
    ConstraintWeights_1:float=0.3414
    ConstraintWeights_2:float=0.1453
    ConstraintWeights_3:float=0.0432
    ConstraintWeights_4:float=0.2567
   ) -> (Pose) {
    alphaBoolBlend {blendOption=Linear}
    boneToModify {boneName=dyn_backpack}
    constraintSetup [
     {targetBone {boneName=clavicle_r}}
     {targetBone {boneName=spine_03}}
     {targetBone {boneName=spine_04}}
     {targetBone {boneName=spine_05}}
     {targetBone {boneName=clavicle_l}}
    ]
    constraintWeights [1,1,1,1,1]
   }
   Output Pose(Result=Component To Local(ComponentPose=@n3)) {layerGroup=None}
  ]}
 EventGraph {nodes [
   Event BlueprintUpdateAnimation()
   n1 = Get Skel_Mesh()
   > Set Shoulder_rot_f_r(Shoulder_rot_f_r:double=GetAlphaFromReferenceTransformDelta(SkeletalMeshComponent=@n1, BoneName=upperarm_r, BaseName=clavicle_r, InputRange=(X=0,Y=90), TwistAxis="0,0,1")) -> (Output_Get:double)
   > Set Shoulder_rot_f_l(Shoulder_rot_f_l:double=GetAlphaFromReferenceTransformDelta(SkeletalMeshComponent=@n1, BoneName=upperarm_l, BaseName=clavicle_l, InputRange=(X=0,Y=90), TwistAxis="0,0,1")) -> (Output_Get:double)
   Event BlueprintInitializeAnimation()
   > Set Skel_Mesh(Skel_Mesh=GetOwnerSkeletalMeshComponent())
   TryGetPawnOwner()
  ]}
]
variables [
 Shoulder_rot_f_l {type=number}
 Shoulder_rot_f_r {type=number}
 Skel_Mesh {type=SkeletalMeshComponent}
]
